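# Skeletal mesh material for the ML Deformer Framework that displays a heat map from vertex color data, usable with mesh deformers.
# The vertex color red channel, passed through a channel mask parameter, becomes the horizontal coordinate for sampling the engine's HeatmapGradient debug texture.
# A smoothstep on the same value blends from neutral grey to the gradient color, so low values stay grey and higher ones show heat colors.
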Material /MLDeformerFramework/Materials/MLDeformerHeatMapMat
graph {
 properties {bUsedWithMeshDeformer=true, bUsedWithSkeletalMesh=true, pixelDepthOffsetMode=PDOM_Legacy, usedShadingModels=Default Lit}
 nodes [
  n1 = VertexColor()
  n2 = ChannelMaskParameter(None=MaterialFunctionCall(X=@n1.R, Y=@n1.R, Z=@n1.R, A=@n1.R) {materialFunction=&MakeFloat4}.Result)
  n3 = Constant() {r=0.5}
  LinearInterpolate(
   A=Constant3Vector() {constant=<rgba 0.5 0.5 0.5 0>}
   B=TextureSample(UVs=MaterialFunctionCall(X=@n2.R, Y=@n3) {materialFunction=&MakeFloat2}.Result) {texture=/Engine/EngineDebugMaterials/HeatmapGradient}.RGB
   Alpha=SmoothStep(Value=@n2.R) {constMax=0.3, constMin=0.2}
  )
 ]
}
